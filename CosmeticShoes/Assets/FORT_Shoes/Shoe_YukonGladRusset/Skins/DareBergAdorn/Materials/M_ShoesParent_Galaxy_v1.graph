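# Parent material for the DareBergAdorn galaxy cosmetic shoe skin, built on the 2023 character parent's PreFX and PostFX functions.
# It adjusts roughness and specular, zeroes metallic, applies MF_Galaxy_v1, blends in a tinted emissive logo, then runs MF_Sock.
# Incoming metallic doubles as the logo mask, and tech-art channels supply curvature, cloth and sole masks.
# An inverted normal-difference and Fresnel mask drives temporal responsiveness, with a debug emissive overlay.
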Material /CosmeticShoes/Assets/FORT_Shoes/Shoe_YukonGladRusset/Skins/DareBergAdorn/Materials/M_ShoesParent_Galaxy_v1
graph {
 properties {
  bAutomaticallySetUsageInEditor=false
  bHasPixelAnimation=true
  bUseMaterialAttributes=true
  bUsedWithClothing=true
  bUsedWithMeshDeformer=true
  bUsedWithMorphTargets=true
  bUsedWithNiagaraMeshParticles=true
  bUsedWithSkeletalMesh=true
  pixelDepthOffsetMode=PDOM_Legacy
  shadingModel=MSM_FromMaterialExpression
  usedShadingModels=Default Lit | Subsurface | Clear Coat | Subsurface Profile | Cloth
 }
 nodes [
  n1 = MaterialFunctionCall() {materialFunction=&MF_CharacterParent_2023_PreFX}
  n2 = NamedRerouteDeclaration(None=@n1.Tech Art Mask) {name="TechArt Mask (V3)", nodeColor=<rgba 1 0 0.2344 1>}
  n3 = ChannelMaskParameter(None=@n2) {group=DareBergAdorn Shoe | Textures, maskChannel=Blue, parameterName=CurvatureFXMaskChannel, sortPriority=20, defaultValue=<rgba 0 0 1 0>}
  NamedRerouteDeclaration(None=@n3.R) {name="Curvature Mask (S)", nodeColor=<rgba 1 0 0.8203 1>}
  n4 = NamedRerouteUsage()
  n5 = NamedRerouteUsage()
  n6 = TextureSampleParameter2D() {
   group=00_Texture Inputs
   parameterName=Normals
   samplerSource=SSM_Wrap_WorldGroupSettings
   samplerType=SAMPLERTYPE_Normal
   sortPriority=2
   texture=&FlatNormalBlue
  }
  n7 = Distance(
   A=Normalize(VectorInput=MaterialFunctionCall(Normal=@n4, Flatness=@n5) {materialFunction=&FlattenNormal}.Result)
   B=@n6.RGB
  )
  n8 = NamedRerouteUsage()
  n9 = Fresnel(Normal=VertexNormalWS()) {baseReflectFraction=0, exponent=4}
  n10 = Saturate(None=Max(A=Max(A=@n7, B=@n8), B=Multiply(A=@n9) {constB=2}))
  n11 = ScalarParameter() {defaultValue=1, group=DareBergAdorn Shoe | Temporal Responsiveness, parameterName=Temporal Mask Exp, sliderMax=10, sortPriority=30}
  NamedRerouteDeclaration(None=Max(A=Power(Base=@n10, Exp=@n11), B=NamedRerouteUsage())) {name="Temporal Mask (S)", nodeColor=<rgba 1 0.7266 0 1>}
  n12 = GetMaterialAttributes(None=@n1.Result) {attributeGetTypes [#1, #2]}
  NamedRerouteDeclaration(None=@n12.Metallic) {name="Logo Mask (S)", nodeColor=<rgba 0 1 0.4609 1>}
  NamedRerouteDeclaration(None=@n12.Base Color) {name="BaseColor In (V3)", nodeColor=<rgba 0.4297 1 0 1>}
  n13 = VectorParameter() {group=DareBergAdorn Shoe | Logo Emissive, parameterName=Logo Emissive Tint, sortPriority=20, channelNames {a=n/a, b="", g="", r=""}}
  n14 = Multiply(A=@n13.RGB, B=NamedRerouteUsage())
  n15 = ScalarParameter() {defaultValue=1, group=DareBergAdorn Shoe | Logo Emissive, parameterName=Logo Emissive Brightness, sortPriority=30}
  n16 = TextureSampleParameter2D() {
   group=DareBergAdorn Shoe | Textures
   parameterName=Normals - No Stitches
   samplerSource=SSM_Wrap_WorldGroupSettings
   samplerType=SAMPLERTYPE_Normal
   sortPriority=10
   texture=&FlatNormalBlue
  }
  n17 = NamedRerouteDeclaration(None=@n16.RGB) {name="Normal No Stitches TS (V3)", nodeColor=<rgba 0.2656 1 0 1>}
  n18 = Fresnel(
   ExponentIn=ScalarParameter() {group=DareBergAdorn Shoe | Logo Emissive, parameterName=Logo Emissive Fresnel Exp, sortPriority=40}
   Normal=Normalize(VectorInput=Transform(None=@n17))
  ) {baseReflectFraction=0}
  n19 = ScalarParameter() {group=DareBergAdorn Shoe | Logo Emissive, parameterName=Logo Emissive Fresnel Intensity, sliderMax=1, sortPriority=50}
  NamedRerouteDeclaration(None=Multiply(A=Multiply(A=@n14, B=@n15), B=LinearInterpolate(B=@n18, Alpha=@n19) {constA=1})) {name="Logo Emissive Out (V3)", nodeColor=<rgba 0.5234 1 0 1>}
  NamedRerouteDeclaration(None=Multiply(
    A=NamedRerouteUsage()
    B=ScalarParameter() {defaultValue=1, group=DareBergAdorn Shoe | Logo Emissive, parameterName=Logo Diffuse Brightness, sortPriority=60}
   )) {name="Logo BaseColor Out (V3)", nodeColor=<rgba 1 0 0.8438 1>}
  n20 = ScalarParameter() {defaultValue=1, group=DareBergAdorn Shoe | Logo Emissive, parameterName=Logo Emissive Mix, sliderMax=1, sortPriority=10}
  NamedRerouteDeclaration(None=Multiply(A=NamedRerouteUsage(), B=@n20)) {name="Logo Emissive Mask (S)", nodeColor=<rgba 1 0 0.0469 1>}
  NamedRerouteDeclaration(None=ChannelMaskParameter(None=@n2) {group=05_Tech Art Masks, parameterName=ClothFuzzMaskChannel}.R) {name="Cloth Mask (S)", nodeColor=<rgba 1 0 0.5391 1>}
  n21 = Multiply(
   A=NamedRerouteUsage()
   B=ScalarParameter() {group=DareBergAdorn Shoe | Surface, parameterName=Opaque Curvature Mix, sliderMax=1, sortPriority=10}
  )
  n22 = NamedRerouteUsage()
  NamedRerouteDeclaration(None=OneMinus(None=Max(A=@n21, B=@n22))) {name="Galaxy Mask (S)", nodeColor=<rgba 0 0.6484 1 1>}
  NamedRerouteDeclaration(None=ChannelMaskParameter(None=@n2) {group=DareBergAdorn Shoe | Textures, parameterName=SoleFXMaskChannel, sortPriority=30}.R) {name="Sole Mask (S)", nodeColor=<rgba 1 0 0.8203 1>}
  n23 = GetMaterialAttributes(None=@n12.MaterialAttributes) {attributeGetTypes [#3, #4]}
  n24 = ScalarParameter() {defaultValue=1, group=DareBergAdorn Shoe | Surface, parameterName=Surface Specular Mult, sortPriority=20}
  n25 = Multiply(
   A=Power(
    Base=@n23.Roughness
    Exp=ScalarParameter() {defaultValue=1, group=DareBergAdorn Shoe | Surface, parameterName=Surface Roughness Exp, sortPriority=30}
   )
   B=ScalarParameter() {defaultValue=1, group=DareBergAdorn Shoe | Surface, parameterName=Surface Roughness Mult, sortPriority=40}
  )
  n26 = MaterialFunctionCall!(
   Material=SetMaterialAttributes!(MaterialAttributes=@n23.MaterialAttributes, Specular=Multiply(A=@n23.Specular, B=@n24), Roughness=Saturate(None=@n25), Metallic=Constant()) {attributeSetTypes [#3, #4, #1]}
   TechArtMask=NamedRerouteUsage()
  ) {materialFunction=/Game/Global/Materials/Character_Parent_2023/Galaxy_Parent_2025/MF_Galaxy_v1}
  n27 = BlendMaterialAttributes(
   A=@n26.Result
   B=SetMaterialAttributes!(MaterialAttributes=@n26.Result, Base Color=NamedRerouteUsage(), Emissive Color=NamedRerouteUsage()) {attributeSetTypes [#2, #5]}
   Alpha=NamedRerouteUsage()
  ) {vertexAttributeBlendType=UseA}
  n28 = NamedRerouteUsage()
  n29 = GetMaterialAttributes(None=MaterialFunctionCall(
    Material=MaterialFunctionCall!(InOpacity, Material=@n27, Mask=@n28) {materialFunction=/Game/Global/Materials/Cosmetics/Kicks/Socks/Materials/MF_Sock}.Result
    M=@n1.M
    Pre Skinned Local Position=@n1.M
    Distance Blend=@n1.M
    Fresnel=@n1.M
   ) {materialFunction=&MF_CharacterParent_2023_PostFX}.Result) {attributeGetTypes [#5]}
  n30 = NamedRerouteUsage()
  n31 = ScalarParameter() {group=DareBergAdorn Shoe | Temporal Responsiveness, parameterName=Debug Temporal Mask, sliderMax=1, sortPriority=40}
  SetMaterialAttributes!(
   MaterialAttributes=@n29.MaterialAttributes
   Emissive Color=LinearInterpolate(A=@n29.Emissive Color, B=Convert!() {convertOutputs [{type=Vector3, defaultValue=<rgba 1 0 0 1>}]}, Alpha=Multiply(A=@n30, B=@n31))
  ) {attributeSetTypes [#5]}
  n32 = ScalarParameter() {defaultValue=1, group=DareBergAdorn Shoe | Temporal Responsiveness, parameterName=Temporal Mask Mix, sliderMax=10, sortPriority=20}
  n33 = Multiply(A=NamedRerouteUsage(), B=@n32)
  n34 = ScalarParameter() {group=DareBergAdorn Shoe | Temporal Responsiveness, parameterName=Temporal Responsiveness Mult, sliderMax=1, sortPriority=10}
  TemporalResponsivenessOutput!(None=Multiply(A=OneMinus(None=@n33), B=@n34))
 ]
}
